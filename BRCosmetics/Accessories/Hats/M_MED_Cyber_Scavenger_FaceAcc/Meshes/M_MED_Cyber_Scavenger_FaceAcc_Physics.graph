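# Physics asset for the Cyber Scavenger face accessory mesh, defining simulation bodies for its dynamic hair bangs and spikes.
# Three bang bones and three spike bones each carry a single rotated box body, while the hair attach bone carries a small sphere.
# All six hair bodies constrain directly to that attach body in a star layout rather than chains, with tight per-piece swing and twist limits that allow only slight secondary motion.
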PhysicsAsset /BRCosmetics/Accessories/Hats/M_MED_Cyber_Scavenger_FaceAcc/Meshes/M_MED_Cyber_Scavenger_FaceAcc_Physics
bodies {
 dyn_bangs_01 [{shapeType=Box, extentX=2.5, extentY=9.51, extentZ=1.5, center=<0.6036 0 -0.2197>, rotation=<rot 0 -89.9999 20.0002>}]
 dyn_bangs_02 [{shapeType=Box, extentX=3.8424, extentY=1.3685, extentZ=9.06, center=<1.2279 0.4161 -1.3285>, rotation=<rot 0 -92.7479 -49.059>}]
 dyn_bangs_03 [{shapeType=Box, extentX=3.5141, extentY=6.1696, extentZ=8.8744, center=<3.2334 1.287 -3.2824>, rotation=<rot 0 -48.8776 -53.3031>}]
 dyn_hair_attach [{shapeType=Sphere, radius=0.75}]
 dyn_spikes_01 [{shapeType=Box, extentX=8.01, extentY=3.01, extentZ=3.5, center=<0.9509 1.4684 2.7937>, rotation=<rot -41.3777 -119.3426 -2.8271>}]
 dyn_spikes_02 [{shapeType=Box, extentX=2.5, extentY=2.5, extentZ=7.01, center=<1.0028 0 1.4283>, rotation=<rot 0 -89.9999 39.9997>}]
 dyn_spikes_03 [{shapeType=Box, extentX=3.01, extentY=6, extentZ=3.01, center=<0.2607 0.1007 0.8319>, rotation=<rot 0 -90 -50.0006>}]
}
constraints [
 {bone1Name=dyn_bangs_02, bone2Name=dyn_hair_attach, swing1LimitDegrees=3, swing2LimitDegrees=10, twistLimitDegrees=5}
 {bone1Name=dyn_bangs_01, bone2Name=dyn_hair_attach, swing1LimitDegrees=15, swing2LimitDegrees=15, twistLimitDegrees=10}
 {bone1Name=dyn_spikes_01, bone2Name=dyn_hair_attach, swing1LimitDegrees=8, swing2LimitDegrees=1, twistLimitDegrees=2}
 {bone1Name=dyn_spikes_02, bone2Name=dyn_hair_attach, swing1LimitDegrees=7, swing2LimitDegrees=10, twistLimitDegrees=2}
 {bone1Name=dyn_spikes_03, bone2Name=dyn_hair_attach, swing1LimitDegrees=8, swing2LimitDegrees=9, twistLimitDegrees=2}
 {bone1Name=dyn_bangs_03, bone2Name=dyn_hair_attach, swing1LimitDegrees=4, swing2LimitDegrees=8, twistLimitDegrees=2}
]
