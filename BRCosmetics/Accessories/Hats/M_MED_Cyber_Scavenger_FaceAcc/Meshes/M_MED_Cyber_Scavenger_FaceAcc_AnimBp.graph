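# Animation Blueprint for the Cyber Scavenger face accessory, which follows the parent mesh's pose and adds secondary motion.
# After copying pose and curves, it blends mask jaw rotation between head and jaw, then runs position constraints on necklace bones.
# Head offset from spine_05, measured each update, drives the alpha of a small additive shift on a rear necklace bone.
# A final RigidBody node simulates physics in spine_01 space with damped, clamped component motion.
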AnimBlueprint /BRCosmetics/Accessories/Hats/M_MED_Cyber_Scavenger_FaceAcc/Meshes/M_MED_Cyber_Scavenger_FaceAcc_AnimBp
target_skeleton=&/M_MED_Cyber_Scavenger_FaceAccSkeleton
graphs [
 AnimGraph {nodes [
   n1 = "Constraint - C_mask_jaw (RR)"!(
    ComponentPose=Local To Component(LocalPose=Copy Pose From Mesh() {bCopyCurves=true, bUseAttachedParent=true})
    Alpha=1
    ConstraintWeights_0=0.6
    ConstraintWeights_1:float=0.4
   ) -> (Pose) {
    alphaBoolBlend {blendOption=Linear}
    boneToModify {boneName=C_mask_jaw}
    constraintSetup [{transformType=Rotation, targetBone {boneName=head}}, {transformType=Rotation, targetBone {boneName=C_jaw}}]
    constraintWeights [1,1]
   }
   n2 = "Constraint - const_necklace_front (P)"!(ComponentPose=@n1, Alpha=1, ConstraintWeights_0=1, ConstraintWeights_1:float=0) -> (Pose) {
    alphaBoolBlend {blendOption=Linear}
    boneToModify {boneName=const_necklace_front}
    constraintSetup [{targetBone {boneName=const_necklace_front}}]
    constraintWeights [1,0]
   }
   n3 = "Constraint - const_necklace_right (P)"!(ComponentPose=@n2, Alpha=1, ConstraintWeights_0=1, ConstraintWeights_1:float=0) -> (Pose) {
    alphaBoolBlend {blendOption=Linear}
    boneToModify {boneName=const_necklace_right}
    constraintSetup [{targetBone {boneName=const_necklace_right}}]
    constraintWeights [1,0]
   }
   n4 = "Constraint - const_necklace_left (P)"!(ComponentPose=@n3, Alpha=1, ConstraintWeights_0=1, ConstraintWeights_1:float=0) -> (Pose) {
    alphaBoolBlend {blendOption=Linear}
    boneToModify {boneName=const_necklace_left}
    constraintSetup [{targetBone {boneName=const_necklace_left}}]
    constraintWeights [1,0]
   }
   n5 = "Constraint - const_necklace_back (P)"!(ComponentPose=@n4, Alpha=1, ConstraintWeights_0=1, ConstraintWeights_1:float=0) -> (Pose) {
    alphaBoolBlend {blendOption=Linear}
    boneToModify {boneName=const_necklace_back}
    constraintSetup [{targetBone {boneName=const_necklace_back}}]
    constraintWeights [1,0]
   }
   n6 = "Transform (Modify) Bone - Bone: const_necklace_back_2"(ComponentPose=@n5, Alpha=Get head_rot_forward() -> (head_rot_forward:double), Translation="0,-0.35,0", Rotation="0,0,0", Scale="1,1,1") {
    rotationMode=BMM_Additive
    rotationSpace=BCS_BoneSpace
    translationMode=BMM_Additive
    translationSpace=BCS_BoneSpace
    alphaBoolBlend {blendOption=Linear}
    boneToModify {boneName=const_necklace_back_2}
    translation {y=-0.35}
   }
   n7 = RigidBody(ComponentPose=@n6) {
    bClampLinearTranslationLimitToRefPose=true
    bOverrideWorldGravity=true
    overridePhysicsAsset=&/M_MED_Cyber_Scavenger_FaceAcc_Physics
    simulationSpace=BaseBoneSpace
    alphaBoolBlend {blendOption=Linear}
    baseBoneRef {boneName=spine_01}
    componentAppliedLinearAccClamp=<6000 6000 6000>
    componentLinearAccScale=<0.75 0.75 0.75>
    componentLinearVelScale=<0.75 0.75 0.75>
    overrideWorldGravity {z=-980}
   }
   Output Pose(Result=Component To Local(ComponentPose=@n7)) {layerGroup=None}
  ]}
 EventGraph {nodes [
   Event BlueprintUpdateAnimation()
   n1 = BreakTransform(InTransform=GetDeltaTransformFromRefPose(self=GetOwnerSkeletalMeshComponent(), BoneName=head, BaseName=spine_05))
   > Set head_rot_forward(head_rot_forward:double=MapRangeClamped(Value=BreakVector(InVec=@n1).Y, InRangeA=-4, InRangeB=0)) -> (Output_Get:double)
   TryGetPawnOwner()
  ]}
]
variables [head_rot_forward {type=number}]
